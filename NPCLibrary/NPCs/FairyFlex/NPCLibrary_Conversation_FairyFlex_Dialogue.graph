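# Conversation database for the FairyFlex NPC's dialogue: optional lore chat about Hades, Cerberus and the gods, voiced as Hades' spouse.
# A single entry point opens a question node, whose greeting changes to a "seedling" line for players wearing listed banana-themed outfits via a controller CID requirement.
# Its "Ask about..." choice branches into three topics plus Back.
# Each topic is a linear chain of Speech nodes converging on one shared ReturnToStart node.
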ConversationDatabase /NPCLibrary/NPCs/FairyFlex/NPCLibrary_Conversation_FairyFlex_Dialogue
entry_points [{tag=Conversation.Entry.FairyFlex.Dialogue, destinations [#1]}]
nodes [
 ConversationEntryPointNode_0 {
  id=#1
  class=ConversationEntryPointNode
  outputs [#2]
  properties {entryTag {tagName=Conversation.Entry.FairyFlex.Dialogue}}
 }
 Speech_C_0 {
  id=#2
  class=Speech_C
  outputs [#3, #4, #5, #6]
  properties {
   nodeName=""
   softTaskIcon=&T_UI_NPC_question
   generalConfig {
    defaultMessage=What's on your mind?
    contextualMessages [{
      message="Well then, seedling, what would you like to know?"
      contextRequirements [{requirement {
         class=FortControllerRequirement_HasCID
         properties {allowedCIds [
           &CID_349_Athena_Commando_M_Banana
           /BRCosmetics/Athena/Items/Cosmetics/Characters/CID_602_Athena_Commando_M_NanaSplit
           &CID_701_Athena_Commando_M_BananaAgent
           &CID_814_Athena_Commando_M_BananaSummer
           /BRCosmetics/Athena/Items/Cosmetics/Characters/CID_A_259_Athena_Commando_M_PeelyTech
           &CID_A_323_Athena_Commando_M_BananaWinter
           /BRCosmetics/Athena/Items/Cosmetics/Characters/CID_A_355_Athena_Commando_M_PeelyToon
           /BRCosmetics/Athena/Items/Cosmetics/Characters/CID_A_437_Athena_Commando_M_ChiselMashup
           &Character_BananaAdventure
          ]}
        }}]
      priority {curve {rowName=Tandem.Speech.Priority.OutfitReaction}}
     }]
   }
  }
  sub_nodes [ConversationChoiceNode_5 {class=ConversationChoiceNode, properties {defaultChoiceDisplayText=Ask about...}}]
 }
 Speech_C_19 {
  id=#6
  class=Speech_C
  outputs [#7]
  properties {
   nodeName=""
   softTaskIcon=&T_UI_NPC_question
   generalConfig {defaultMessage=I don't know why the pantheon has a reputation for infighting.}
  }
  sub_nodes [ConversationChoiceNode_0 {class=ConversationChoiceNode, properties {defaultChoiceDisplayText=The Gods}}]
 }
 Speech_C_20 {
  id=#7
  class=Speech_C
  outputs [#8]
  properties {
   nodeName=""
   softTaskIcon=&T_UI_NPC_question
   generalConfig {defaultMessage="Sure, my husband doesn't like my mother..."}
  }
 }
 Speech_C_21 {
  id=#8
  class=Speech_C
  outputs [#9]
  properties {
   nodeName=""
   softTaskIcon=&T_UI_NPC_question
   generalConfig {defaultMessage=...and Poseidon doesn't like Medusa...}
  }
 }
 Speech_C_22 {
  id=#9
  class=Speech_C
  outputs [#10]
  properties {
   nodeName=""
   softTaskIcon=&T_UI_NPC_question
   generalConfig {defaultMessage="...And Zeus doesn't like... anybody, really."}
  }
 }
 Speech_C_23 {
  id=#10
  class=Speech_C
  outputs [#11]
  properties {
   nodeName=""
   softTaskIcon=&T_UI_NPC_question
   generalConfig {defaultMessage="But we all get along when commanded or forced!"}
  }
 }
 ReturnToStart_C_23 {
  id=#11
  class=ReturnToStart_C
  properties {softTaskIcon=&T_UI_NPC_Back}
  sub_nodes [ConversationChoiceNode_32 {class=ConversationChoiceNode, properties {defaultChoiceDisplayText=BACK}}]
 }
 Speech_C_13 {
  id=#5
  class=Speech_C
  outputs [#12]
  properties {
   nodeName=""
   softTaskIcon=&T_UI_NPC_question
   generalConfig {defaultMessage="All dogs are good, but my dog is the best."}
  }
  sub_nodes [ConversationChoiceNode_4 {class=ConversationChoiceNode, properties {defaultChoiceDisplayText=Cerberus}}]
 }
 Speech_C_16 {
  id=#12
  class=Speech_C
  outputs [#13]
  properties {
   nodeName=""
   softTaskIcon=&T_UI_NPC_question
   generalConfig {defaultMessage=If anyone ever harmed a hair on his snout...}
  }
 }
 Speech_C_17 {
  id=#13
  class=Speech_C
  outputs [#14]
  properties {
   nodeName=""
   softTaskIcon=&T_UI_NPC_question
   generalConfig {defaultMessage=...They'd suffer the full wrath of the Underworld.}
  }
 }
 Speech_C_18 {
  id=#14
  class=Speech_C
  outputs [#11]
  properties {
   nodeName=""
   softTaskIcon=&T_UI_NPC_question
   generalConfig {defaultMessage="Luckily, you'd never do anything like that!"}
  }
 }
 Back_C_0 {
  id=#4
  class=Back_C
  sub_nodes [ConversationChoiceNode_7 {class=ConversationChoiceNode, properties {defaultChoiceDisplayText=BACK}}]
 }
 Speech_C_7 {
  id=#3
  class=Speech_C
  outputs [#15]
  properties {
   nodeName=""
   softTaskIcon=&T_UI_NPC_question
   generalConfig {defaultMessage=Most Gods marry for love or for power. I married for both.}
  }
  sub_nodes [ConversationChoiceNode_3 {class=ConversationChoiceNode, properties {defaultChoiceDisplayText=Hades}}]
 }
 Speech_C_15 {
  id=#15
  class=Speech_C
  outputs [#16]
  properties {
   nodeName=""
   softTaskIcon=&T_UI_NPC_question
   generalConfig {defaultMessage=I can still remember the day we eloped...}
  }
 }
 Speech_C_14 {
  id=#16
  class=Speech_C
  outputs [#17]
  properties {
   nodeName=""
   softTaskIcon=&T_UI_NPC_question
   generalConfig {defaultMessage="The ceremony was in the Underworld, surrounded by beloved wailing spirits."}
  }
 }
 Speech_C_11 {
  id=#17
  class=Speech_C
  outputs [#18]
  properties {
   nodeName=""
   softTaskIcon=&T_UI_NPC_question
   generalConfig {defaultMessage=The air smelled of moist soil and decay. Charon cried so hard his jaw fell off.}
  }
 }
 Speech_C_12 {
  id=#18
  class=Speech_C
  outputs [#11]
  properties {nodeName="", softTaskIcon=&T_UI_NPC_question, generalConfig {defaultMessage=It was... Perfect. *sigh*}}
 }
]
